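# Idle Niagara effect for the CyberDelivery pickaxe, adding a persistent soft capsule glow to the weapon.
# One sprite emitter bursts two particles, using spawn groups to give each its own size, offset and alignment.
# User.Variant selects the tint, while front-end modifiers and a time-of-day scalar adjust brightness and opacity.
# An owner-scale threshold check switches the alpha on or off.
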NiagaraSystem /BRCosmetics/Weapons/FORT_Melee/Pickaxe_CyberDelivery/FX/NS_Pickaxe_CyberDelivery_Idle
dependencies {
 usedDataInterfaces [&NiagaraDataInterfaceColorCurve]
 usedDynamicInputs [&ColorFromCurve, &FNTimeOfDayScalar]
 usedModules [
  &SystemState
  &self:SystemSpawnScript.SystemScriptSource.SystemScriptGraph.NiagaraNodeAssignment_1.SetVariables_#1
 ]
}
emitters {CapsuleGlow {
  emitterUpdateScript [
   &EmitterState(
    Inactive Response=::NewEnumerator1
    Life Cycle Mode=::NewEnumerator1
    Loop Behavior=::NewEnumerator1
    Loop Duration Mode=::NewEnumerator1
    UseLoopDelay=false
   )
   &SpawnBurst_Instantaneous(Spawn Count=2, Use Spawn Probability=false, UseLoopCountLimit=true)
  ]
  particleSpawnScript [
   &FN_SetSpawnGroup(Number of Spawn Groups=2)
   &InitializeParticle(
    Sprite Rotation Mode=::NewEnumerator0
    Sprite Size=&MergeHelper_Vector2Switch(0=<18 69.4865>, 1=<18 30.7>, 2=?, 3=?, 4=?, Index=&SpawnGroup(), Num=::NewEnumerator0)
    Sprite Size Mode=::NewEnumerator4
    UsePositionOffset=true
    Write Material Random=true
   )
   &SpriteFacingAndAlignment(
    Sprite Alignment=&MergeHelper_VectorSwitch(0=<0.0525 0 1>, 1=<1.0696 0 1>, 2=?, 3=?, Index=&SpawnGroup(), Num=::NewEnumerator0)
    Sprite Facing=<0.7068 0.07 0.3>
    WriteAlignment=true
    WriteFacing=false
   )
   &SystemLocation(
    MaskSpawnGroup=false
    Offset=&MergeHelper_VectorSwitch(0=<-1.8662 0 37.2271>, 1=<3 0 68>, 2=?, 3=?, Index=&SpawnGroup(), Num=::NewEnumerator0)
   )
   &CameraOffset()
  ]
  particleUpdateScript [
   &ParticleState(Kill Particles When Lifetime Has Elapsed=false, Let Infinitely Lived Particles Die When Emitter Deactivates=true)
   &ScaleColor(
    Scale Alpha=&Multiply_Float(
     A=&Multiply_Float(A=&FN_FrontEndFloatModifier(Gameplay=0.1, Front End=0.035), B=$System.TODFloat)
     B=&MakeCustomFloatFromBool(
      Bool=&SetBoolByFloatComparison(A=0.9, B=&VectorLength(Vector=$Engine.Owner.Scale))
      True Float=0
      False Float=1
     )
    )
    Scale RGB=&MakeVectorFromLinearColorRGB(Color=&MergeHelper_ColorSwitch(
      0=&Multiply_LinearColorByFloat(Linear Color=<rgba 1 0.0462 0.03 1>, Float=&FN_FrontEndFloatModifier(Gameplay=1, Front End=0.65))
      1=&Multiply_LinearColorByFloat(Linear Color=<rgba 1 0.0462 0.03 1>, Float=&FN_FrontEndFloatModifier(Gameplay=1, Front End=0.65))
      2=<rgba 0.9271 0.178 0 1>
      3=<rgba 0.2874 0.5156 0.5156 1>
      4=<rgba 0.0705 0.5208 0.5009 1>
      5=&Multiply_LinearColorByFloat(
       Linear Color=<rgba 0.4379 0.1214 0.8906 1>
       Float=&FN_FrontEndFloatModifier(Gameplay=1, Front End=1.3)
      )
      Index=&MakeIntFromFloat(Float=$User.Variant)
      Num=::NewEnumerator4
     ))
    ScaleA=true
    ScaleRGB=true
   )
  ]
  rendererClasses [&NiagaraSpriteRendererProperties]
  renderers [{rendererClass=&NiagaraSpriteRendererProperties}]
 }}
summary {
 systemName=NS_Pickaxe_CyberDelivery_Idle
 emitters [{emitterName=CapsuleGlow, rendererClasses [&NiagaraSpriteRendererProperties]}]
 userVariables [
  User.bIsFrontEnd {type=NiagaraBool}
  User.bIsFrontEndPreview {type=NiagaraBool}
  User.bisSwinging {type=NiagaraBool}
  User.Variant {type=NiagaraFloat, defaultValue=0, description=""}
 ]
}
