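# Niagara impact effect for the CyberDelivery pickaxe, with emitters including a directional flare, sparks, streaks, flash elements and smoke.
# A system direction vector pushes the directional emitters, and User.Variant selects per-style colors, spark counts and a variant-specific flash shape.
# Flash_Elements packs spike, star, flare, hex, shape and light into one emitter through custom per-renderer attributes.
# Front-end modifiers suppress the directional bursts and shrink or dim other elements.
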NiagaraSystem /BRCosmetics/Weapons/FORT_Melee/Pickaxe_CyberDelivery/FX/NS_Pickaxe_CyberDelivery_Impact
dependencies {
 usedDataInterfaces [
  &NiagaraDataInterfaceCurve
  &NiagaraDataInterfaceColorCurve
  &NiagaraDataInterfaceArrayColor
  &NiagaraDataInterfaceArrayFloat2
  &NiagaraDataInterfaceVector2DCurve
  &NiagaraDataInterfaceArrayFloat
  &NiagaraDataInterfaceSpriteRendererInfo
  &NiagaraDataInterfaceArrayFloat3
 ]
 usedDynamicInputs [
  &Vector4FromCurve
  &UniformRangedVector2D
  &SetBoolByIntComparison
  &SetBoolByFloatComparison
  /Niagara/DynamicInputs/Math/Clamp_Vector
  &FN_FrontEndVector3Modifier
  &SelectVectorFromArray
 ]
 usedModules [
  &self:SystemSpawnScript.SystemScriptSource.SystemScriptGraph.NiagaraNodeAssignment_0.SetVariables_#1
  &self:SystemSpawnScript.SystemScriptSource.SystemScriptGraph.NiagaraNodeAssignment_2.SetVariables_#2
  &self:SystemSpawnScript.SystemScriptSource.SystemScriptGraph.NiagaraNodeAssignment_3.SetVariables_#3
  &self:SystemSpawnScript.SystemScriptSource.SystemScriptGraph.NiagaraNodeAssignment_4.SetVariables_#4
  &self:SystemSpawnScript.SystemScriptSource.SystemScriptGraph.NiagaraNodeAssignment_5.SetVariables_#5
  &self:SystemSpawnScript.SystemScriptSource.SystemScriptGraph.NiagaraNodeAssignment_6.SetVariables_#6
  &self:SystemSpawnScript.SystemScriptSource.SystemScriptGraph.NiagaraNodeAssignment_7.SetVariables_#7
  &SystemState
  &self:SystemSpawnScript.SystemScriptSource.SystemScriptGraph.NiagaraNodeAssignment_1.SetVariables_#8
 ]
}
emitters {
 DirectionalFlare {
  emitterUpdateScript [
   &EmitterState(
    Enable Distance Culling=true
    Life Cycle Mode=::NewEnumerator1
    Loop Behavior=::NewEnumerator1
    Loop Delay=0.01
    Loop Duration=1
    Max Distance Response=::KillImmediately
    MaxDistance=3500
    Module.bUseMaxDistance=true
    Module.bUseMinDistance=false
    Scalability Mode=::NewEnumerator1
    UseLoopDelay=false
   )
   &SpawnBurst_Instantaneous(
    Spawn Count=1
    Spawn Probability=&FN_FrontEndFloatModifier(Gameplay=1, Front End=0)
    Spawn Time=0.005
    Use Spawn Probability=true
   )
  ]
  particleSpawnScript [
   &InitializeParticle(
    Color=&MergeHelper_ColorSwitch(
     0=<rgba 6 0.27 0.2127 1>
     1=<rgba 6 0.27 0.2127 1>
     2=<rgba 6 2.9342 0 1>
     3=<rgba 1.1247 5.9382 9 1>
     4=<rgba 1.1075 4.1159 6 1>
     5=<rgba 1.464 0.8077 6 1>
     Index=&MakeIntFromFloat(Float=$User.Variant, Conversion Mode=::NewEnumerator1)
     Num=::NewEnumerator4
    )
    Lifetime=0.26
    Sprite Size=<15 500>
    Sprite Size Mode=::NewEnumerator4
    UsePositionOffset=false
    Write Material Random=true
   ) {module=InitializeParticle001}
   &CameraOffset(Camera Offset Amount=40)
   &DynamicMaterialParameters(
    Index 0 Param 1=0.2
    Index 0 Param 2=1
    Module.Param0WriteEnabled=true
    Module.Param1WriteEnabled=true
    Module.Param2WriteEnabled=true
    Module.Param3WriteEnabled=true
   )
   &AddVelocity(Velocity=$System.DirectionVector, Velocity Speed Scale=430)
  ]
  particleUpdateScript [
   &ParticleState()
   &SolveForcesAndVelocity(Clamp Velocity=false, Limit Acceleration=false)
   &ScaleColor(
    Color Value To Scale=$Particles.Color
    Scale Alpha=&FloatFromCurve(FloatCurve=DataInterface() {propertyValues {Curve {keys [
         <at 0.02 0 RCIM_Cubic>
         {arriveTangent=-2.9108, interpMode=RCIM_Cubic, leaveTangent=-2.9108, tangentMode=RCTM_User, time=0.06, value=1}
         {interpMode=RCIM_Cubic, tangentMode=RCTM_User, time=1, value=0}
        ]}}})
    Scale RGB=&VectorFromFloat(Value=&Lerp_Float(
      A=5
      B=0.6
      Alpha=&FloatFromCurve(FloatCurve=DataInterface() {propertyValues {Curve {keys [
           {arriveTangent=7.4609, interpMode=RCIM_Cubic, leaveTangent=7.4609, tangentMode=RCTM_User, time=0, value=0}
           {arriveTangent=0.0212, interpMode=RCIM_Cubic, leaveTangent=0.0212, tangentMode=RCTM_User, time=0.4, value=1}
          ]}}})
     ))
    ScaleA=true
    ScaleRGB=true
   )
   &ScaleSpriteSize(
    Non-Uniform Curve Sprite Scale=DataInterface() {propertyValues {
      XCurve {keys [
        {interpMode=RCIM_Cubic, tangentMode=RCTM_User, time=0, value=0.2}
        {arriveTangent=3.4491, interpMode=RCIM_Cubic, leaveTangent=3.4491, tangentMode=RCTM_User, time=0.08, value=0.3}
        {arriveTangent=-2.6795, interpMode=RCIM_Cubic, leaveTangent=-2.6795, tangentMode=RCTM_User, time=0.11, value=1.2}
        {arriveTangent=-1.3483, interpMode=RCIM_Cubic, leaveTangent=-1.3483, time=0.2, value=0.6}
        {interpMode=RCIM_Cubic, tangentMode=RCTM_User, time=1, value=0}
       ]}
      YCurve {keys [
        {interpMode=RCIM_Cubic, tangentMode=RCTM_User, time=0, value=0.2}
        {interpMode=RCIM_Cubic, tangentMode=RCTM_User, time=0.1, value=1.025}
        {tangentMode=RCTM_User, time=1, value=0.8}
       ]}
     }}
    Scale Sprite Size Mode=::NewEnumerator3
   ) {module=ScaleSpriteSize001}
   /SetVariables(
    Particles.ColorDark:LinearColor=&MergeHelper_ColorSwitch(
     0=&ColorFromCurve(ColorCurve=DataInterface() {propertyValues {
        AlphaCurve {keys [<at 0 2 RCIM_Cubic>,<at 0.091 0.7>,<at 0.519 0 RCIM_Cubic>]}
        BlueCurve {keys [<at 0 0 RCIM_Cubic>]}
        GreenCurve {keys [<at 0 0.002 RCIM_Cubic>]}
        RedCurve {keys [<at 0 0.06 RCIM_Cubic>]}
       }})
     1=&ColorFromCurve(ColorCurve=DataInterface() {propertyValues {
        AlphaCurve {keys [<at 0 2 RCIM_Cubic>,<at 0.091 0.7>,<at 0.519 0 RCIM_Cubic>]}
        BlueCurve {keys [<at 0 0 RCIM_Cubic>]}
        GreenCurve {keys [<at 0 0.002 RCIM_Cubic>]}
        RedCurve {keys [<at 0 0.06 RCIM_Cubic>]}
       }})
     2=&ColorFromCurve(ColorCurve=DataInterface() {propertyValues {
        AlphaCurve {keys [<at 0 2 RCIM_Cubic>,<at 0.091 0.7>,<at 0.519 0 RCIM_Cubic>]}
        BlueCurve {keys [<at 0 0.0012 RCIM_Cubic>]}
        GreenCurve {keys [<at 0 0.0272 RCIM_Cubic>]}
        RedCurve {keys [<at 0 0.06 RCIM_Cubic>]}
       }})
     3=&ColorFromCurve(ColorCurve=DataInterface() {propertyValues {
        AlphaCurve {keys [<at 0 2 RCIM_Cubic>,<at 0.091 0.7>,<at 0.519 0 RCIM_Cubic>]}
        BlueCurve {keys [<at 0 0.1042 RCIM_Cubic>]}
        GreenCurve {keys [<at 0 0.0053 RCIM_Cubic>]}
        RedCurve {keys [<at 0 0.0625 RCIM_Cubic>]}
       }})
     4=&ColorFromCurve(ColorCurve=DataInterface() {propertyValues {
        AlphaCurve {keys [<at 0 2 RCIM_Cubic>,<at 0.091 0.7>,<at 0.519 0 RCIM_Cubic>]}
        BlueCurve {keys [<at 0 0.06 RCIM_Cubic>]}
        GreenCurve {keys [<at 0 0.0353 RCIM_Cubic>]}
        RedCurve {keys [<at 0 0.0072 RCIM_Cubic>]}
       }})
     5=&ColorFromCurve(ColorCurve=DataInterface() {propertyValues {
        AlphaCurve {keys [<at 0 2 RCIM_Cubic>,<at 0.091 0.7>,<at 0.519 0 RCIM_Cubic>]}
        BlueCurve {keys [<at 0 0.06 RCIM_Cubic>]}
        GreenCurve {keys [<at 0 0.0083 RCIM_Cubic>]}
        RedCurve {keys [<at 0 0.0472 RCIM_Cubic>]}
       }})
     Index=&MakeIntFromFloat(Float=$User.Variant, Conversion Mode=::NewEnumerator1)
     Num=::NewEnumerator4
    )
    Particles.SpriteSizeDark:Vector2f=&Multiply_Vector2D(A=$Particles.SpriteSize, B=<0.8 0.7>)
   ) {isSetParameters=true}
  ]
  rendererClasses [&NiagaraSpriteRendererProperties]
  renderers [
   {rendererClass=&NiagaraSpriteRendererProperties}
   {rendererClass=&NiagaraSpriteRendererProperties, rendererIndex=1}
  ]
 }
 DirectionalSparks {
  emitterUpdateScript [
   &EmitterState(Life Cycle Mode=::NewEnumerator1, Loop Behavior=::NewEnumerator1, Loop Delay=0.05, Loop Duration=1, UseLoopDelay=true)
   &SpawnBurst_Instantaneous(
    Spawn Count=&MergeHelper_IntSwitch(
     0=&UniformRangedInt(Minimum=7, Maximum=15)
     1=&UniformRangedInt(Minimum=7, Maximum=15)
     2=&UniformRangedInt(Minimum=4, Maximum=8)
     3=&UniformRangedInt(Minimum=4, Maximum=8)
     4=0
     5=&UniformRangedInt(Minimum=4, Maximum=8)
     Index=&MakeIntFromFloat(Float=$User.Variant, Conversion Mode=::NewEnumerator1)
     Num=::NewEnumerator4
    )
    Spawn Probability=&FN_FrontEndFloatModifier(Gameplay=1, Front End=0)
    Use Spawn Probability=true
    UseLoopCountLimit=false
   )
   &SpawnBurst_Instantaneous(
    Spawn Count=&UniformRangedInt(Minimum=6, Maximum=8)
    Spawn Group=1
    Spawn Probability=&FN_FrontEndFloatModifier(Gameplay=1, Front End=0)
    Use Spawn Probability=true
    UseLoopCountLimit=false
   ) {module=SpawnBurst_Instantaneous001}
  ]
  particleSpawnScript [
   &InitializeParticle(
    Color=&MergeHelper_ColorSwitch(
     0=<rgba 5 0.15 0.15 1>
     1=<rgba 5 0.15 0.15 1>
     2=<rgba 5 1.0792 0.2146 1>
     3=&RandomRangeLinearColor(Minimum=<rgba 0.1358 4.5611 7 1>, Maximum=<rgba 6.4334 0.2355 7 1>)
     4=<rgba 0.5573 0.8802 1 1>
     5=&RandomRangeLinearColor(Minimum=<rgba 0.3341 0.0202 5 1>, Maximum=<rgba 1.1743 0.1532 5 1>)
     Index=&MakeIntFromFloat(Float=$User.Variant, Conversion Mode=::NewEnumerator1)
     Num=::NewEnumerator4
    )
    Lifetime Max=0.45
    Lifetime Min=0.05
    Lifetime Mode=::NewEnumerator1
    Mesh Scale Mode=::NewEnumerator3
    Mesh Uniform Scale=&RandomRangeFloat(Minimum=0.1, Maximum=0.5)
    Sprite Size Max=<2 3>
    Sprite Size Min=<1 2>
    Sprite Size Mode=::NewEnumerator2
    UsePositionOffset=false
    Write Material Random=true
   ) {module=InitializeParticle001}
   &ShapeLocation(
    MaskSpawnGroup=false
    Override Seed=false
    Rotation Axis=$User.Direction
    Rotation Mode=::NewEnumerator1
    Sphere Radius=10
    UseOwnerScale=true
   )
   &CameraOffset(Camera Offset Amount=15)
   &AddVelocityFromPoint(Module.Use Falloff Distance=false, Velocity Strength=&RandomRangeFloat(Minimum=300, Maximum=400))
   &AddVelocity(Velocity=$System.DirectionVector, Velocity Speed Scale=&RandomRangeFloat(Minimum=300, Maximum=600))
   &InitialMeshOrientation(Mesh Orientation Mode=::NewEnumerator7, Rotation=<0 0.7 0>, Use Rotation Vector=true)
   /SetVariables(Particles.VisibilityTag=&SpawnGroup()) {isSetParameters=true}
  ]
  particleUpdateScript [
   &ParticleState()
   &AccelerationForce(Acceleration=<0 0 -1100>)
   &Drag(Drag=&RandomRangeFloat(Minimum=2.5, Maximum=5), Use Linear Drag=true, Use Rotational Drag=true)
   &SolveForcesAndVelocity(Clamp Velocity=false, Limit Acceleration=false)
   &ScaleSpriteSizeBySpeed(Max Scale Factor=<1 10>, Min Scale Factor=<0.65 0.65>)
   &ScaleSpriteSize(
    Scale Sprite Size Mode=::NewEnumerator0
    Uniform Scale Factor=&FloatFromCurve(FloatCurve=DataInterface() {propertyValues {Curve {keys [{arriveTangent=-6.725, interpMode=RCIM_Cubic, leaveTangent=-6.725, tangentMode=RCTM_User, time=0.65, value=1}, <at 1 0 RCIM_Cubic>]}}})
   )
   &ScaleMeshSize(Scale Factor=&VectorFromFloat(Value=&FloatFromCurve(FloatCurve=DataInterface() {propertyValues {Curve {keys [
          {interpMode=RCIM_Cubic, tangentMode=RCTM_User, time=0, value=0}
          {interpMode=RCIM_Cubic, tangentMode=RCTM_User, time=0.1, value=1}
          {arriveTangent=-1.4364, interpMode=RCIM_Cubic, leaveTangent=-1.4364, tangentMode=RCTM_User, time=1, value=0}
         ]}}})))
   &ScaleColor(
    Linear Color Curve=DataInterface() {propertyValues {
      AlphaCurve {keys [<at 0.38 1>,<at 1 0>]}
      BlueCurve {keys [<at 0 15>,<at 1 1>]}
      GreenCurve {keys [<at 0 15>,<at 1 1>]}
      RedCurve {keys [<at 0 15>,<at 1 1>]}
     }}
    Scale Mode=::NewEnumerator2
   )
   /SetVariables(Particles.IceColor:LinearColor=&Multiply_LinearColorByFloat(
     Linear Color=&RandomRangeLinearColor(Minimum=<rgba 0.5573 0.8802 1 1>, Maximum=<rgba 1.1146 1.7603 2 1>)
     Float=&FNTimeOfDayScalar(Night Scale=0.65, Morning Scale=0.88, Evening Scale=0.8)
    )) {isSetParameters=true}
  ]
  rendererClasses [&NiagaraSpriteRendererProperties, &NiagaraMeshRendererProperties]
  renderers [
   {rendererClass=&NiagaraSpriteRendererProperties}
   {rendererClass=&NiagaraMeshRendererProperties, rendererIndex=1}
  ]
 }
 DirectionalStreaks {
  emitterUpdateScript [
   &EmitterState()
   &SpawnBurst_Instantaneous(
    Spawn Count=3
    Spawn Probability=&FN_FrontEndFloatModifier(Gameplay=1, Front End=0)
    Use Spawn Probability=true
    UseLoopCountLimit=false
   )
  ]
  particleSpawnScript [
   &InitializeParticle(
    Color=&MergeHelper_ColorSwitch(
     0=<rgba 1 0.13 0.1 1>
     1=<rgba 1 0.13 0.1 1>
     2=<rgba 1 0.5619 0.1218 1>
     3=&RandomRangeLinearColor(Minimum=<rgba 0.1551 1.335 2 1>, Maximum=<rgba 1.6692 0.2399 2 1>)
     4=<rgba 0.0474 0.574 1 1>
     5=&RandomRangeLinearColor(Minimum=<rgba 0.0948 0.3076 1 1>, Maximum=<rgba 0.2965 0.1585 1 1>)
     Index=&MakeIntFromFloat(Float=$User.Variant, Conversion Mode=::NewEnumerator1)
     Num=::NewEnumerator4
    )
    Lifetime Max=0.2
    Lifetime Min=0.14
    Lifetime Mode=::NewEnumerator1
    Sprite Size Max=<25 80>
    Sprite Size Min=<20 60>
    Sprite Size Mode=::NewEnumerator2
    Sprite UV Mode=::NewEnumerator1
    UsePositionOffset=false
    Write Material Random=true
   ) {module=InitializeParticle001}
   &AddVelocity(Module.Use Falloff Distance=false, Override Seed=false, Velocity Mode=::NewEnumerator1) {module=AddVelocity001}
   &SubUVAnimation(End Frame=3, SubUV Animation Mode:ENiagara_SubUVLookupMode=::NewEnumerator2)
   &AddVelocity(
    Velocity=$System.DirectionVector
    Velocity Speed Scale=&MakeCustomFloatFromBool(Bool=$User.bIsFrontEndPreview, True Float=0, False Float=&RandomRangeFloat(Minimum=60, Maximum=100))
   )
   &CameraOffset(Camera Offset Amount=25)
  ]
  particleUpdateScript [
   &ParticleState()
   &ScaleSpriteSize(Uniform Curve Sprite Scale=DataInterface() {propertyValues {Curve {keys [{tangentMode=RCTM_User, time=0, value=0}, {tangentMode=RCTM_User, time=0.2, value=1}]}}}) {module=ScaleSpriteSize001}
   &SolveForcesAndVelocity(Clamp Velocity=false, Limit Acceleration=false)
   &ScaleSpriteSize(
    Scale Factor=&Vector2DFromCurve(
     Vector2Curve=DataInterface() {propertyValues {
       XCurve {keys [{arriveTangent=-2.3004, interpMode=RCIM_Cubic, leaveTangent=-2.3004, tangentMode=RCTM_User, time=0, value=1}, <at 1 0.1 RCIM_Cubic>]}
       YCurve {keys [{arriveTangent=1.8515, interpMode=RCIM_Cubic, leaveTangent=1.8515, tangentMode=RCTM_User, time=0, value=0.3}, <at 1 1.5 RCIM_Cubic>]}
      }}
     ScaleCurve=&Multiply_Vector2D_ByFloat(
      Float=&MakeCustomFloatFromBool(Bool=$User.bIsFrontEndPreview, True Float=0.6, False Float=1)
      Vector2D=<1.5 1>
     )
    )
    Scale Sprite Size Mode=::NewEnumerator2
   )
   &ScaleColor(
    Scale Alpha=&FloatFromCurve(
     FloatCurve=DataInterface() {propertyValues {Curve {keys [<at 0 0 RCIM_Cubic>,<at 0.1 1 RCIM_Cubic>,<at 1 0>]}}}
     Scale Curve=&FNTimeOfDayScalar(Night Scale=0.5, Morning Scale=1, Evening Scale=0.5)
    )
    Scale RGB=&VectorFromFloat(Value=&FloatFromCurve(FloatCurve=DataInterface() {propertyValues {Curve {keys [
          {arriveTangent=-52.7997, interpMode=RCIM_Cubic, leaveTangent=-52.7997, tangentMode=RCTM_User, time=0, value=10}
          {arriveTangent=-3.0339, interpMode=RCIM_Cubic, leaveTangent=-3.0339, tangentMode=RCTM_User, time=0.2, value=1.5}
          <at 1 1 RCIM_Cubic>
         ]}}}))
    ScaleA=true
    ScaleRGB=true
   )
  ]
  rendererClasses [&NiagaraSpriteRendererProperties]
  renderers [{rendererClass=&NiagaraSpriteRendererProperties}]
 }
 Flash_Elements {
  emitterUpdateScript [
   &EmitterState()
   &SpawnBurst_Instantaneous(Spawn Group=1, Use Spawn Probability=false, UseLoopCountLimit=false) {module=SpawnBurst_Instantaneous001}
   &SpawnBurst_Instantaneous(
    Spawn Count=&RandomRangeInt(Minimum=6, Maximum=8)
    Use Spawn Probability=false
    UseLoopCountLimit=false
   )
  ]
  particleSpawnScript [
   &InitializeParticle(Color Mode=::NewEnumerator0, Lifetime=0.23, UsePositionOffset=false, Write Material Random=true)
   &ShapeLocation(MaskSpawnGroup=false, Override Seed=false, Sphere Radius=1, UseOwnerScale=true)
   &SpriteFacingAndAlignment(Sprite Alignment=$Particles.ShapeLocation.ShapeNormal, WriteAlignment=true, WriteFacing=false)
   &FN_PerParticleLOD(Enable SpriteSize Distance Scaling=true, Sprite Size Scale Far=<2.5 2.5>)
   /SetVariables(Particles.SpriteSize=<1 1>, Particles.VisibilityTag=&SpawnGroup()) {isSetParameters=true}
   /SetVariables(
    Particles.Flash Spike Color:LinearColor=&MergeHelper_ColorSwitch(
     0=<rgba 0.8 0.0683 0.016 1>
     1=<rgba 0.8 0.0683 0.016 1>
     2=<rgba 0.8 0.4799 0.0411 1>
     3=<rgba 0.3719 1.0804 1.4 1>
     4=<rgba 0.1543 0.7936 0.9271 1>
     5=<rgba 0.3703 0.1668 1 1>
     Index=&MakeIntFromFloat(Float=$User.Variant, Conversion Mode=::NewEnumerator1)
     Num=::NewEnumerator4
    )
    Particles.Flash Spike Lifetime:NiagaraFloat=&RandomRangeFloat(Minimum=0.05, Maximum=0.07)
    Particles.Flash Spike Sprite Size:Vector2f=&Multiply_Vector2D_ByFloat(
     Float=&FN_FrontEndFloatModifier(Gameplay=1, Front End=$System.FrontendScale)
     Vector2D=&RandomRangeVector2D(Minimum=<3 60>, Maximum=<8 80>)
    )
   ) {isSetParameters=true}
   /SetVariables(
    Particles.Flash Star Cam Offset=15
    Particles.Flash Star Lifetime=0.15
    Particles.Flash Star Rotation:NiagaraFloat=&RandomRangeFloat(Minimum=0, Maximum=360)
    Particles.Flash Star Sprite Size:Vector2f=&Multiply_Vector2D_ByFloat(Float=&FN_FrontEndFloatModifier(Gameplay=1, Front End=0.45), Vector2D=$Particles.SpriteSize)
    Particles.Flash Star Sprite Size Alt:Vector2f=&Multiply_Vector2D_ByFloat(
     Float=&Multiply_Float(
      A=0.85
      B=&Multiply_Float(
       A=&FN_FrontEndFloatModifier(Gameplay=1, Front End=0.45)
       B=&RandomRangeFloat(Minimum=120, Maximum=150)
      )
     )
     Vector2D=$Particles.SpriteSize
    )
    Particles.Flash Star SubUV Index:NiagaraFloat=&RandomRangeFloat(Minimum=0, Maximum=3)
    Particles.Flash Star SubUV Index Alt:NiagaraFloat=&RandomRangeFloat(Minimum=0, Maximum=3)
   ) {isSetParameters=true}
   /SetVariables(
    Particles.Flash Flare Color:LinearColor=&Multiply_LinearColorByFloat(
     Linear Color=&MergeHelper_ColorSwitch(
      0=<rgba 3 0.1874 0.06 0.7>
      1=<rgba 3 0.1874 0.06 0.7>
      2=<rgba 3 1.2627 0.1124 0.7>
      3=<rgba 4.0861 1.4345 5 0.7>
      4=<rgba 1.2969 2.7182 3 0.7>
      5=<rgba 1.221 0.4283 3 0.7>
      Index=&MakeIntFromFloat(Float=$User.Variant, Conversion Mode=::NewEnumerator1)
      Num=::NewEnumerator4
     )
     Float=&MakeCustomFloatFromBool(Bool=$User.bIsFrontEnd, True Float=0.5, False Float=1)
    )
    Particles.Flash Flare Lifetime=0.2
    Particles.Flash Flare Sprite Size:Vector2f=&Vector2DFromFloat(Value=&FN_FrontEndFloatModifier(Gameplay=100, Front End=90))
   ) {isSetParameters=true}
   &DynamicMaterialParameters(
    Index 0 Param 1=0.2
    Index 0 Param 2=1
    Module.Index0Float0Write=true
    Module.Index0Float1Write=true
    Module.Index0Float2Write=true
    Module.Index0Float3Write=true
   )
   &CameraOffset(Camera Offset Amount=&FN_FrontEndFloatModifier(Gameplay=0, Front End=10))
   /SetVariables(Particles.SpecificShapeCamOffset=20) {isSetParameters=true}
  ]
  particleUpdateScript [
   &ParticleState()
   /SetVariables(Particles.Flash Flare Sprite Size:Vector2f=&Multiply_Vector2D_ByFloat(
     Float=&FloatFromCurve(FloatCurve=DataInterface() {propertyValues {Curve {keys [
          {arriveTangent=-4.3411, interpMode=RCIM_Cubic, leaveTangent=-4.3411, tangentMode=RCTM_User, time=0, value=1.5}
          {interpMode=RCIM_Cubic, tangentMode=RCTM_User, time=0.45, value=0.4}
         ]}}})
     Vector2D=$Particles.Initial.Flash Flare Sprite Size
    )) {isSetParameters=true}
   /SetVariables(
    Particles.Update Flash Flare Age:NiagaraFloat=&Divide_Float(A=$Particles.Age, B=$Particles.Initial.Flash Flare Lifetime)
    Particles.Update Flash Spike Age:NiagaraFloat=&Divide_Float(A=$Particles.Age, B=$Particles.Initial.Flash Spike Lifetime)
    Particles.Update Flash Star Age:NiagaraFloat=&Divide_Float(A=$Particles.Age, B=$Particles.Initial.Flash Star Lifetime)
   ) {isSetParameters=true}
   /SetVariables(
    Particles.Update Flash Spike Color:LinearColor=&MakeLinearColorFromVectorAndFloat(
     "Vector (RGB)"=&MakeVectorFromLinearColorRGB(Color=$Particles.Initial.Flash Spike Color)
     "Float (Alpha)"=&FloatFromCurve(FloatCurve=DataInterface() {propertyValues {Curve {keys [<at 0.5 1>,<at 1 0>]}}}, CurveIndex=$Particles.Update Flash Spike Age)
    )
    Particles.Update Flash Spike Sprite Size:Vector2f=&Multiply_Vector2D_ByFloat(
     Float=&FloatFromCurve(
      FloatCurve=DataInterface() {propertyValues {Curve {keys [<at 0 0.6>,<at 0.9 2>]}}}
      CurveIndex=$Particles.Update Flash Spike Age
      Scale Curve=&FN_FrontEndFloatModifier(Gameplay=1, Front End=0.6)
     )
     Vector2D=$Particles.Initial.Flash Spike Sprite Size
    )
    Particles.Update Flash Star Color:LinearColor=&MergeHelper_ColorSwitch(
     0=&ColorFromCurve(
      ColorCurve=DataInterface() {propertyValues {
        AlphaCurve {keys [<at 0 1>,<at 0.8992 0>]}
        BlueCurve {keys [<at 0.1451 0.5>,<at 0.4068 0.1188>,<at 0.7329 0.031>]}
        GreenCurve {keys [<at 0.1451 0.975>,<at 0.4068 0.1815>,<at 0.7329 0.031>]}
        RedCurve {keys [<at 0.1451 10>,<at 0.4068 2>,<at 0.7329 1.5>]}
       }}
      CurveIndex=$Particles.Update Flash Star Age
     )
     1=&ColorFromCurve(
      ColorCurve=DataInterface() {propertyValues {
        AlphaCurve {keys [<at 0 1>,<at 0.8992 0>]}
        BlueCurve {keys [<at 0.1451 0.5>,<at 0.4068 0.1188>,<at 0.7329 0.031>]}
        GreenCurve {keys [<at 0.1451 0.975>,<at 0.4068 0.1815>,<at 0.7329 0.031>]}
        RedCurve {keys [<at 0.1451 10>,<at 0.4068 2>,<at 0.7329 1.5>]}
       }}
      CurveIndex=$Particles.Update Flash Star Age
     )
     2=&ColorFromCurve(
      ColorCurve=DataInterface() {propertyValues {
        AlphaCurve {keys [<at 0 1>,<at 0.8992 0>]}
        BlueCurve {keys [<at 0.1451 0.5406>,<at 0.4068 0.1711>,<at 0.7329 0>]}
        GreenCurve {keys [<at 0.1451 2.6609>,<at 0.4068 0.5758>,<at 0.7329 0.3108>]}
        RedCurve {keys [<at 0.1451 10>,<at 0.4068 2>,<at 0.7329 1.5>]}
       }}
      CurveIndex=$Particles.Update Flash Star Age
     )
     3=&ColorFromCurve(
      ColorCurve=DataInterface() {propertyValues {
        AlphaCurve {keys [<at 0 1>,<at 0.8992 0>]}
        BlueCurve {keys [<at 0.1451 10>,<at 0.4068 3>,<at 0.7329 2>]}
        GreenCurve {keys [<at 0.1451 7.4636>,<at 0.4068 0.7008>,<at 0.7329 0.5011>]}
        RedCurve {keys [<at 0.1451 4.1749>,<at 0.4068 0.625>,<at 0.7329 1.6283>]}
       }}
      CurveIndex=$Particles.Update Flash Star Age
     )
     4=&ColorFromCurve(
      ColorCurve=DataInterface() {propertyValues {
        AlphaCurve {keys [<at 0 1>,<at 0.8992 0>]}
        BlueCurve {keys [<at 0.1451 10>,<at 0.4068 2>,<at 0.7329 1.5>]}
        GreenCurve {keys [<at 0.1451 8.6711>,<at 0.4068 1.3794>,<at 0.7329 0.4496>]}
        RedCurve {keys [<at 0.1451 1.7923>,<at 0.4068 0.4072>,<at 0.7329 0.1745>]}
       }}
      CurveIndex=$Particles.Update Flash Star Age
     )
     5=&ColorFromCurve(
      ColorCurve=DataInterface() {propertyValues {
        AlphaCurve {keys [<at 0 1>,<at 0.8992 0>]}
        BlueCurve {keys [<at 0.1451 10>,<at 0.4068 2>,<at 0.7329 1.5>]}
        GreenCurve {keys [<at 0.1451 7.4636>,<at 0.4068 1.2488>,<at 0.7329 0.2733>]}
        RedCurve {keys [<at 0.1451 4.1749>,<at 0.4068 0.4577>,<at 0.7329 0.6318>]}
       }}
      CurveIndex=$Particles.Update Flash Star Age
     )
     Index=&MakeIntFromFloat(Float=$User.Variant, Conversion Mode=::NewEnumerator1)
     Num=::NewEnumerator4
    )
    Particles.Update Flash Star Sprize Size:Vector2f=&Multiply_Vector2D_ByFloat(
     Float=&FloatFromCurve(
      FloatCurve=DataInterface() {propertyValues {Curve {keys [
          {arriveTangent=2.7013, interpMode=RCIM_Cubic, leaveTangent=2.7013, tangentMode=RCTM_User, time=0, value=0.15}
          {interpMode=RCIM_Cubic, tangentMode=RCTM_User, time=1, value=1}
         ]}}}
      CurveIndex=$Particles.Update Flash Star Age
     )
     Vector2D=$Particles.Initial.Flash Star Sprite Size
    )
    Particles.Update Flash Star Sprize Size Alt:Vector2f=&Multiply_Vector2D_ByFloat(
     Float=&FloatFromCurve(
      FloatCurve=DataInterface() {propertyValues {Curve {keys [
          {arriveTangent=2.7013, interpMode=RCIM_Cubic, leaveTangent=2.7013, tangentMode=RCTM_User, time=0, value=0}
          {interpMode=RCIM_Cubic, tangentMode=RCTM_User, time=1, value=1}
         ]}}}
      CurveIndex=$Particles.Update Flash Star Age
     )
     Vector2D=$Particles.Initial.Flash Star Sprite Size Alt
    )
   ) {isSetParameters=true}
   /SetVariables(Particles.Update Flash Flare Color:LinearColor=&MakeLinearColorFromVectorAndFloat(
     "Vector (RGB)"=&Multiply_Vector(
      A=&MakeVectorFromLinearColorRGB(Color=$Particles.Initial.Flash Flare Color)
      B=&VectorFromFloat(Value=&FN_FrontEndFloatModifier(Gameplay=1.5, Front End=0.4))
     )
     "Float (Alpha)"=&Multiply_Float(
      A=&MakeFloatFromLinearColor(LinearColor=<rgba 1 1 1 1>, Channel:ENiagaraLinearColor_Channels=::NewEnumerator3)
      B=&Lerp_Float(
       A=0
       B=10
       Alpha=&FloatFromCurve(FloatCurve=DataInterface() {propertyValues {Curve {keys [
            {arriveTangent=-4.121, interpMode=RCIM_Cubic, leaveTangent=-4.121, tangentMode=RCTM_User, time=0, value=1}
            {interpMode=RCIM_Cubic, tangentMode=RCTM_User, time=0.7, value=0}
           ]}}})
      )
     )
    )) {isSetParameters=true}
   &FN_Light_Attributes(
    Color=&MergeHelper_ColorSwitch(
     0=<rgba 20 0.1 0 1>
     1=<rgba 20 0.1 0 1>
     2=<rgba 20 5.3276 0.0781 1>
     3=<rgba 7.349 15.9016 20 1>
     4=<rgba 2.8367 14.2283 20 1>
     5=<rgba 4.4401 2.5752 20 1>
     Index=&MakeIntFromFloat(Float=$User.Variant, Conversion Mode=::NewEnumerator1)
     Num=::NewEnumerator4
    )
    Create Color Attribute=true
    Create Exponent Attribute=true
    Create Position Attribute=true
    Create Radius Attribute=true
    Exponent=5
    Position=<15 0 0>
    Radius=100
    Scale Color=&FloatFromCurve(FloatCurve=DataInterface() {propertyValues {Curve {keys [
         {arriveTangent=-1.8529, interpMode=RCIM_Cubic, leaveTangent=-1.8529, tangentMode=RCTM_User, time=0, value=1}
         {arriveTangent=-0.5426, interpMode=RCIM_Cubic, leaveTangent=-0.5426, tangentMode=RCTM_User, time=0.8, value=0}
        ]}}})
   )
   /SetVariables(Particles.ColorDark:LinearColor=&Multiply_LinearColorByFloat(
     Linear Color=&MergeHelper_ColorSwitch(
      0=&ColorFromCurve(ColorCurve=DataInterface() {propertyValues {
         AlphaCurve {keys [<at 0.2394 1 RCIM_Cubic>,<at 0.9406 0 RCIM_Cubic>]}
         BlueCurve {keys [<at 0 0.0007 RCIM_Cubic>]}
         GreenCurve {keys [<at 0 0.0007 RCIM_Cubic>]}
         RedCurve {keys [<at 0 0.0417 RCIM_Cubic>]}
        }})
      1=&ColorFromCurve(ColorCurve=DataInterface() {propertyValues {
         AlphaCurve {keys [<at 0.2394 1 RCIM_Cubic>,<at 0.9406 0 RCIM_Cubic>]}
         BlueCurve {keys [<at 0 0.0007 RCIM_Cubic>]}
         GreenCurve {keys [<at 0 0.0007 RCIM_Cubic>]}
         RedCurve {keys [<at 0 0.0417 RCIM_Cubic>]}
        }})
      2=&ColorFromCurve(ColorCurve=DataInterface() {propertyValues {
         AlphaCurve {keys [<at 0.2394 1 RCIM_Cubic>,<at 0.9406 0 RCIM_Cubic>]}
         BlueCurve {keys [<at 0 0.0011 RCIM_Cubic>]}
         GreenCurve {keys [<at 0 0.01 RCIM_Cubic>]}
         RedCurve {keys [<at 0 0.0417 RCIM_Cubic>]}
        }})
      3=&ColorFromCurve(ColorCurve=DataInterface() {propertyValues {
         AlphaCurve {keys [<at 0.2394 1 RCIM_Cubic>,<at 0.9406 0 RCIM_Cubic>]}
         BlueCurve {keys [<at 0 0.0417 RCIM_Cubic>]}
         GreenCurve {keys [<at 0 0.0119 RCIM_Cubic>]}
         RedCurve {keys [<at 0 0.0335 RCIM_Cubic>]}
        }})
      4=&ColorFromCurve(ColorCurve=DataInterface() {propertyValues {
         AlphaCurve {keys [<at 0.2394 1 RCIM_Cubic>,<at 0.9406 0 RCIM_Cubic>]}
         BlueCurve {keys [<at 0 0.0417 RCIM_Cubic>]}
         GreenCurve {keys [<at 0 0.0205 RCIM_Cubic>]}
         RedCurve {keys [<at 0 0.005 RCIM_Cubic>]}
        }})
      5=&ColorFromCurve(ColorCurve=DataInterface() {propertyValues {
         AlphaCurve {keys [<at 0.2394 1 RCIM_Cubic>,<at 0.9406 0 RCIM_Cubic>]}
         BlueCurve {keys [<at 0 0.0417 RCIM_Cubic>]}
         GreenCurve {keys [<at 0 0.0054 RCIM_Cubic>]}
         RedCurve {keys [<at 0 0.009 RCIM_Cubic>]}
        }})
      Index=&MakeIntFromFloat(Float=$User.Variant, Conversion Mode=::NewEnumerator1)
      Num=::NewEnumerator4
     )
     Float=&FN_FrontEndFloatModifier(Gameplay=1, Front End=0.33)
    )) {isSetParameters=true}
   /SetVariables(
    Particles.HexCameraOffset=25
    Particles.HexColor:LinearColor=&MakeLinearColorFromVectorAndFloat(
     "Vector (RGB)"=&MakeVectorFromLinearColorRGB(Color=&MergeHelper_ColorSwitch(
       0=<rgba 1.3 0.11 0.0745 1>
       1=<rgba 1.3 0.11 0.0745 1>
       2=<rgba 1.3 0.5363 0.1693 1>
       3=<rgba 0.4309 0.9819 1.3 1>
       4=<rgba 0.206 0.9325 1.3 1>
       5=<rgba 0.206 0.9325 1.3 1>
       Index=&MakeIntFromFloat(Float=$User.Variant, Conversion Mode=::NewEnumerator1)
       Num=::NewEnumerator4
      ))
     "Float (Alpha)"=&FloatFromCurve(
      FloatCurve=DataInterface() {propertyValues {Curve {keys [<at 0.3482 1 RCIM_Cubic>,<at 1 0>]}}}
      Scale Curve=&FN_FrontEndFloatModifier(Gameplay=0.5, Front End=0.25)
     )
    )
    Particles.HexDynamic:Vector4f=&MakeVector4(
     X=&FloatFromCurve(FloatCurve=DataInterface() {propertyValues {Curve {keys [
          {arriveTangent=0.632, interpMode=RCIM_Cubic, leaveTangent=0.632, tangentMode=RCTM_User, time=0, value=0}
          {arriveTangent=0.5765, interpMode=RCIM_Cubic, leaveTangent=0.5765, tangentMode=RCTM_User, time=1, value=1}
         ]}}})
     Y=0
     Z=0
     W=0
    )
    Particles.HexSize:Vector2f=&Vector2DFromFloat(Value=&Multiply_Float(A=120, B=&FN_FrontEndFloatModifier(Gameplay=1, Front End=0.45)))
   ) {isSetParameters=true}
   /SetVariables(Particles.DarkGlowSpriteSize:Vector2f=&Multiply_Vector2D_ByFloat(
     Float=&FloatFromCurve(
      FloatCurve=DataInterface() {propertyValues {Curve {keys [
          {arriveTangent=-2.6948, interpMode=RCIM_Cubic, leaveTangent=-2.6948, tangentMode=RCTM_User, time=0, value=1.5}
          {interpMode=RCIM_Cubic, tangentMode=RCTM_User, time=0.45, value=0.4}
         ]}}}
      Scale Curve=&Multiply_Float(A=1.3, B=&FN_FrontEndFloatModifier(Gameplay=1, Front End=0.6))
     )
     Vector2D=$Particles.Initial.Flash Flare Sprite Size
    )) {isSetParameters=true}
   /SetVariables(Particles.SpecificShapeColor:LinearColor=&MultiplyLinearColor(
     Linear Color=&MultiplyLinearColor(
      Linear Color=&SelectLinearColorFromArray(
       Array Sampling Mode=::NewEnumerator1
       Direct Array Index=&MakeIntFromFloat(Float=$User.Variant, Conversion Mode=::NewEnumerator1)
       Color Selection Array=DataInterface() {propertyValues {ColorData [<rgba 0.6414 1 0 1>, <rgba 1 1 1 1>, <rgba 0 0 0 4>, <rgba 1 1 1 1>, <rgba 0.6852 0 1 1>, <rgba 1 1 1 1>]}}
       Random Seed=?
      )
      Scale Factor=&MakeVector4_FromLinearColor(LinearColor=&MergeHelper_ColorSwitch(
        0=<rgba 0 0 0 0>
        1=<rgba 0 0 0 0>
        2=&ColorFromCurve(ColorCurve=DataInterface() {propertyValues {
           AlphaCurve {keys [<at 0 1 RCIM_Cubic>,<at 0.349 1>,<at 1 0 RCIM_Cubic>]}
           BlueCurve {keys [<at 0 10 RCIM_Cubic>,<at 0.3714 1>,<at 1 0.2969>]}
           GreenCurve {keys [<at 0 10 RCIM_Cubic>,<at 0.3714 1>,<at 1 0.2969>]}
           RedCurve {keys [<at 0 10 RCIM_Cubic>,<at 0.3714 1>,<at 1 0.2969>]}
          }})
        3=&ColorFromCurve(ColorCurve=DataInterface() {propertyValues {
           AlphaCurve {keys [<at 0 1 RCIM_Cubic>,<at 1 0 RCIM_Cubic>]}
           BlueCurve {keys [<at 0.2179 0>,<at 0.6209 2>]}
           GreenCurve {keys [<at 0.2179 0>,<at 0.6209 0.9709>]}
           RedCurve {keys [<at 0.2179 0>,<at 0.6209 1.852>]}
          }})
        4=&ColorFromCurve(ColorCurve=DataInterface() {propertyValues {
           AlphaCurve {keys [<at 0 1 RCIM_Cubic>,<at 0.349 1>,<at 1 0 RCIM_Cubic>]}
           BlueCurve {keys [<at 0 10 RCIM_Cubic>,<at 0.3714 1>,<at 1 0.2969>]}
           GreenCurve {keys [<at 0 10 RCIM_Cubic>,<at 0.3714 1>,<at 1 0.2969>]}
           RedCurve {keys [<at 0 10 RCIM_Cubic>,<at 0.3714 1>,<at 1 0.2969>]}
          }})
        5=&ColorFromCurve(ColorCurve=DataInterface() {propertyValues {
           AlphaCurve {keys [<at 0 1 RCIM_Cubic>,<at 0.349 1>,<at 1 0 RCIM_Cubic>]}
           BlueCurve {keys [<at 0 10 RCIM_Cubic>,<at 0.3714 1>,<at 1 0.2969>]}
           GreenCurve {keys [<at 0 10 RCIM_Cubic>,<at 0.3714 1>,<at 1 0.2969>]}
           RedCurve {keys [<at 0 10 RCIM_Cubic>,<at 0.3714 1>,<at 1 0.2969>]}
          }})
        Index=&MakeIntFromFloat(Float=$User.Variant, Conversion Mode=::NewEnumerator1)
        Num=::NewEnumerator4
       ))
     )
     Scale Factor=&MakeVector4_FromLinearColor(LinearColor=<rgba 1 1 1 1>)
    )) {isSetParameters=true}
   /SetVariables(Particles.SpecificShapeSize:Vector2f=&Multiply_Vector2D(
     A=&Multiply_Vector2D_ByFloat(
      Float=&FloatFromCurve(
       FloatCurve=DataInterface() {propertyValues {Curve {keys [{arriveTangent=3.0363, interpMode=RCIM_Cubic, leaveTangent=3.0363, tangentMode=RCTM_User, time=0, value=0}, <at 1 1>]}}}
       Scale Curve=&FN_FrontEndFloatModifier(Gameplay=200, Front End=120)
      )
      Vector2D=<1 1>
     )
     B=&SelectVector2DFromArray(
      Direct Array Index=&MakeIntFromFloat(Float=$User.Variant, Conversion Mode=::NewEnumerator1)
      Vector 2D Selection Array=DataInterface() {propertyValues {FloatData [<1 1>, <1 1>, <0.4 0.7>, <0.5 0.5>, <0.55 0.55>, <0.55 0.55>]}}
     )
    )) {isSetParameters=true}
   /SetVariables(Particles.SpecificShapeDynMat:Vector4f=&MakeVector4(
     X=&MergeHelper_FloatSwitch(
      0=0
      1=0
      2=0
      3=&FloatFromCurve(FloatCurve=DataInterface() {propertyValues {Curve {keys [
           {arriveTangent=1.5862, interpMode=RCIM_Cubic, leaveTangent=1.5862, tangentMode=RCTM_User, time=0, value=0}
           {arriveTangent=0.159, interpMode=RCIM_Cubic, leaveTangent=0.159, tangentMode=RCTM_User, time=1, value=0.8}
          ]}}})
      4=0
      5=0
      Index=&MakeIntFromFloat(Float=$User.Variant, Conversion Mode=::NewEnumerator1)
      Num=::NewEnumerator4
     )
     Y=&MergeHelper_FloatSwitch(
      0=0
      1=0
      2=0
      3=&FloatFromCurve(FloatCurve=DataInterface() {propertyValues {Curve {keys [
           {arriveTangent=2.5938, interpMode=RCIM_Cubic, leaveTangent=2.5938, tangentMode=RCTM_User, time=0, value=0.2}
           {interpMode=RCIM_Cubic, tangentMode=RCTM_User, time=0.2, value=0.5}
           {tangentMode=RCTM_User, time=1, value=0}
          ]}}})
      4=0
      5=0
      Index=&MakeIntFromFloat(Float=$User.Variant, Conversion Mode=::NewEnumerator1)
      Num=::NewEnumerator4
     )
     Z=&MergeHelper_FloatSwitch(
      0=0
      1=0
      2=0
      3=&FloatFromCurve(FloatCurve=DataInterface() {propertyValues {Curve {keys [{interpMode=RCIM_Cubic, tangentMode=RCTM_User, time=0, value=1}, {arriveTangent=-0.3137, tangentMode=RCTM_User, time=1, value=0.8}]}}})
      4=0
      5=0
      Index=&MakeIntFromFloat(Float=$User.Variant, Conversion Mode=::NewEnumerator1)
      Num=::NewEnumerator4
     )
     W=&MergeHelper_FloatSwitch(
      0=0
      1=0
      2=0
      3=&Add_Float(
       A=&RandomRangeFloat(Minimum=0, Maximum=0)
       B=&FloatFromCurve(
        FloatCurve=DataInterface() {propertyValues {Curve {keys [{interpMode=RCIM_Cubic, tangentMode=RCTM_User, time=0, value=0}, {interpMode=RCIM_Cubic, tangentMode=RCTM_User, time=1, value=1}]}}}
        Scale Curve=0.25
       )
      )
      4=0
      5=0
      Index=&MakeIntFromFloat(Float=$User.Variant, Conversion Mode=::NewEnumerator1)
      Num=::NewEnumerator4
     )
    )) {isSetParameters=true}
  ]
  rendererClasses [&NiagaraSpriteRendererProperties, &NiagaraLightRendererProperties]
  renderers [
   {rendererClass=&NiagaraSpriteRendererProperties}
   {rendererClass=&NiagaraSpriteRendererProperties, rendererIndex=1}
   {rendererClass=&NiagaraSpriteRendererProperties, rendererIndex=2}
   {rendererClass=&NiagaraSpriteRendererProperties, rendererIndex=3}
   {rendererClass=&NiagaraLightRendererProperties, rendererIndex=4}
   {rendererClass=&NiagaraSpriteRendererProperties, rendererIndex=5}
   {rendererClass=&NiagaraSpriteRendererProperties, rendererIndex=6}
   {rendererClass=&NiagaraSpriteRendererProperties, rendererIndex=7}
   {rendererClass=&NiagaraSpriteRendererProperties, rendererIndex=8}
  ]
 }
 Impact_Smoke {
  emitterUpdateScript [
   &EmitterState()
   &SpawnBurst_Instantaneous(
    Spawn Count=&RandomRangeInt(Minimum=2, Maximum=3)
    Use Spawn Probability=false
    UseLoopCountLimit=false
   )
  ]
  particleSpawnScript [
   &InitializeParticle(
    Color Maximum=<rgba 0.76 1 1 0.2>
    Color Minimum=<rgba 0.8 0.917 1 0.15>
    Color Mode=::NewEnumerator2
    Lifetime Max=0.4
    Lifetime Min=0.3
    Lifetime Mode=::NewEnumerator1
    Sprite Size Mode=::NewEnumerator3
    Uniform Sprite Size=&Multiply_Float(
     A=&FN_FrontEndFloatModifier(Gameplay=1, Front End=0.45)
     B=&RandomRangeFloat(Minimum=100, Maximum=130)
    )
    UsePositionOffset=false
    Write Material Random=true
   )
   &CameraOffset(Camera Offset Amount=&RandomRangeFloat(Minimum=40, Maximum=50, Evaluation Type=::NewEnumerator1))
   &ShapeLocation(
    Hemisphere Distribution=<1 0.5>
    MaskSpawnGroup=false
    Override Seed=false
    Rotation Angle=90
    Rotation Mode=::NewEnumerator1
    Sphere Radius=&FN_FrontEndFloatModifier(Gameplay=10, Front End=5)
    UseOwnerScale=true
   )
   &AddVelocity(Velocity=&Multiply_VectorByFloat(Vector=$Particles.ShapeLocation.ShapeVector, Float=10))
  ]
  particleUpdateScript [
   &ParticleState()
   &ScaleSpriteSize(Uniform Curve Sprite Scale=DataInterface() {propertyValues {Curve {keys [
        {arriveTangent=1.7761, interpMode=RCIM_Cubic, leaveTangent=1.7761, tangentMode=RCTM_User, time=0, value=0}
        {interpMode=RCIM_Cubic, tangentMode=RCTM_User, time=1, value=1}
       ]}}})
   &ScaleColor(
    Scale Alpha=&FloatFromCurve(FloatCurve=DataInterface() {propertyValues {Curve {keys [
         {arriveTangent=2.4358, interpMode=RCIM_Cubic, leaveTangent=2.4358, tangentMode=RCTM_User, time=0, value=0}
         <at 0.25 0.5 RCIM_Cubic>
         <at 1 0 RCIM_Cubic>
        ]}}})
    Scale RGB=&MakeVectorFromLinearColorRGB(Color=&MergeHelper_ColorSwitch(
      0=&ColorFromCurve(ColorCurve=DataInterface() {propertyValues {
         AlphaCurve {keys [<at 0 1 RCIM_Cubic>]}
         BlueCurve {keys [<at 0.2265 0.3 RCIM_Cubic>,<at 0.7544 1 RCIM_Cubic>]}
         GreenCurve {keys [<at 0.2265 0.3 RCIM_Cubic>,<at 0.7544 1 RCIM_Cubic>]}
         RedCurve {keys [<at 0.2265 1 RCIM_Cubic>,<at 0.7544 1 RCIM_Cubic>]}
        }})
      1=&ColorFromCurve(ColorCurve=DataInterface() {propertyValues {
         AlphaCurve {keys [<at 0 1 RCIM_Cubic>]}
         BlueCurve {keys [<at 0.2265 0.3 RCIM_Cubic>,<at 0.7544 1 RCIM_Cubic>]}
         GreenCurve {keys [<at 0.2265 0.3 RCIM_Cubic>,<at 0.7544 1 RCIM_Cubic>]}
         RedCurve {keys [<at 0.2265 1 RCIM_Cubic>,<at 0.7544 1 RCIM_Cubic>]}
        }})
      2=&ColorFromCurve(ColorCurve=DataInterface() {propertyValues {
         AlphaCurve {keys [<at 0 1 RCIM_Cubic>]}
         BlueCurve {keys [<at 0.2265 0.2772 RCIM_Cubic>,<at 0.7544 1 RCIM_Cubic>]}
         GreenCurve {keys [<at 0.2265 0.4509 RCIM_Cubic>,<at 0.7544 1 RCIM_Cubic>]}
         RedCurve {keys [<at 0.2265 1 RCIM_Cubic>,<at 0.7544 1 RCIM_Cubic>]}
        }})
      3=&ColorFromCurve(ColorCurve=DataInterface() {propertyValues {
         AlphaCurve {keys [<at 0 1 RCIM_Cubic>]}
         BlueCurve {keys [<at 0.2265 1 RCIM_Cubic>,<at 0.7544 1 RCIM_Cubic>]}
         GreenCurve {keys [<at 0.2265 0.7255 RCIM_Cubic>,<at 0.7544 1 RCIM_Cubic>]}
         RedCurve {keys [<at 0.2265 0.4505 RCIM_Cubic>,<at 0.7544 1 RCIM_Cubic>]}
        }})
      4=&ColorFromCurve(ColorCurve=DataInterface() {propertyValues {
         AlphaCurve {keys [<at 0 1 RCIM_Cubic>]}
         BlueCurve {keys [<at 0.2265 1 RCIM_Cubic>,<at 0.7544 1 RCIM_Cubic>]}
         GreenCurve {keys [<at 0.2265 0.7255 RCIM_Cubic>,<at 0.7544 1 RCIM_Cubic>]}
         RedCurve {keys [<at 0.2265 0.4505 RCIM_Cubic>,<at 0.7544 1 RCIM_Cubic>]}
        }})
      5=&ColorFromCurve(ColorCurve=DataInterface() {propertyValues {
         AlphaCurve {keys [<at 0 1 RCIM_Cubic>]}
         BlueCurve {keys [<at 0.2265 1 RCIM_Cubic>,<at 0.7544 1 RCIM_Cubic>]}
         GreenCurve {keys [<at 0.2265 0.2708 RCIM_Cubic>,<at 0.7544 1 RCIM_Cubic>]}
         RedCurve {keys [<at 0.2265 0.3297 RCIM_Cubic>,<at 0.7544 1 RCIM_Cubic>]}
        }})
      Index=&MakeIntFromFloat(Float=$User.Variant, Conversion Mode=::NewEnumerator1)
      Num=::NewEnumerator4
     ))
    ScaleA=true
    ScaleRGB=true
   )
   &DynamicMaterialParameters(
    Index 0 Param 1=&Add_Float(
     A=&RandomRangeFloat(Minimum=0, Maximum=1)
     B=&FloatFromCurve(
      FloatCurve=DataInterface() {propertyValues {Curve {keys [
          <at 0 0 RCIM_Cubic>
          {arriveTangent=1.5947, interpMode=RCIM_Cubic, leaveTangent=1.5947, tangentMode=RCTM_User, time=0.2557, value=0.5257}
          <at 1 1 RCIM_Cubic>
         ]}}}
      CurveIndex=$Particles.Age
      Scale Curve=0.25
     )
    )
    Module.Param0WriteEnabled=true
    Module.Param1WriteEnabled=true
    Module.Param2WriteEnabled=true
    Module.Param3WriteEnabled=true
   ) {module=DynamicMaterialParameters001}
   &SolveForcesAndVelocity(Clamp Velocity=false, Limit Acceleration=false)
  ]
  rendererClasses [&NiagaraSpriteRendererProperties]
  renderers [
   {rendererClass=&NiagaraSpriteRendererProperties}
   {rendererClass=&NiagaraSpriteRendererProperties, rendererIndex=1}
  ]
 }
 NE_Dust {
  emitterUpdateScript [
   &EmitterState(
    Enable Distance Culling=true
    Enable Visibility Culling=true
    MaxDistance=2500
    Module.bApplySpawnCountScaleByDistance=true
    Module.bUseMaxDistance=true
    Module.bUseMinDistance=false
    Scalability Mode=::NewEnumerator1
    Scale Spawn Count=true
   )
   &SpawnBurst_Instantaneous(
    Spawn Count=&RandomRangeInt(Minimum=12, Maximum=18)
    Spawn Probability=&MakeCustomFloatFromBool(Bool=&Boolean_OrOperation(Bool1=$System.IsGold, Bool2=$System.IsGalaxy), True Float=1, False Float=0)
    Use Spawn Probability=true
    UseLoopCountLimit=false
   )
   &SpawnBurst_Instantaneous(
    Spawn Count=&RandomRangeInt(Minimum=5, Maximum=6)
    Spawn Group=1
    Use Spawn Probability=false
    UseLoopCountLimit=false
   ) {module=SpawnBurst_Instantaneous001}
  ]
  particleSpawnScript [
   &InitializeParticle(
    Color=&LerpLinearColors(
     StartColor=&SelectLinearColorFromArray(
      Array Sampling Mode=::NewEnumerator1
      Direct Array Index=&MakeIntFromFloat(Float=$User.Variant, Conversion Mode=::NewEnumerator1)
      Color Selection Array=DataInterface() {propertyValues {ColorData [<rgba 1 1 1 1>, <rgba 1 1 1 1>, <rgba 16 7.925 2.4 1>, <rgba 0.3979 0.626 1 1>, <rgba 0.3979 0.626 1 1>, <rgba 0.3979 0.626 1 1>]}}
      Random Seed=?
     )
     EndColor=&SelectLinearColorFromArray(
      Array Sampling Mode=::NewEnumerator1
      Direct Array Index=&MakeIntFromFloat(Float=$User.Variant, Conversion Mode=::NewEnumerator1)
      Color Selection Array=DataInterface() {propertyValues {ColorData [<rgba 1 1 1 1>, <rgba 1 1 1 1>, <rgba 16 11.0365 2.464 1>, <rgba 1 0.3931 0.9046 1>, <rgba 1 0.3931 0.9046 1>, <rgba 1 0.3931 0.9046 1>]}}
      Random Seed=?
     )
     LerpFactor=&UniformAOrBFloat(A=0, Evaluation Type=::NewEnumerator0, Recalculate Random Each Loop=true, Override Seed=false)
    )
    Lifetime=&RandomRangeFloat(Minimum=0.3, Maximum=0.6)
    Sprite Rotation Angle=&MergeHelper_FloatSwitch(
     0=0
     1=0
     2=0
     3=0
     4=&RandomRangeFloat(Minimum=0, Maximum=1)
     5=0
     Index=&MakeIntFromFloat(Float=$User.Variant, Conversion Mode=::NewEnumerator1)
     Num=::NewEnumerator4
    )
    Sprite Rotation Mode=::NewEnumerator1
    Sprite Size Mode=::NewEnumerator3
    Uniform Sprite Size=&MaskFloatBySpawnGroup(
     Masked Float=&RandomRangeFloat(Minimum=1.5, Maximum=3)
     Passthrough Float=&Multiply_Float(
      A=&RandomRangeFloat(Minimum=5, Maximum=8)
      B=&SelectFloatFromArray(
       Direct Array Index=&MakeIntFromFloat(Float=$User.Variant, Conversion Mode=::NewEnumerator1)
       Float Selection Array=DataInterface() {propertyValues {FloatData [0,0,0.8,1.5,1.2,1.5]}}
      )
     )
    )
    UsePositionOffset=true
    Write Material Random=true
   )
   &ShapeLocation(
    MaskSpawnGroup=false
    Override Seed=false
    Sphere Radius=&Multiply_Float(
     A=&Multiply_Float(
      A=&MaskFloatBySpawnGroup(Masked Float=20, Passthrough Float=50)
      B=&FN_FrontEndFloatModifier(Gameplay=1, Front End=$System.FrontendScale)
     )
     B=&SelectFloatFromArray(
      Direct Array Index=&MakeIntFromFloat(Float=$User.Variant, Conversion Mode=::NewEnumerator1)
      Float Selection Array=DataInterface() {propertyValues {FloatData [1,1,1,1,0.3,1]}}
     )
    )
    Transform Order:ENiagara_TransformOrder=::NewEnumerator1
    UseOwnerScale=true
   ) {module=ShapeLocation001}
   &CameraOffset(Camera Offset Amount=&FN_FrontEndFloatModifier(Gameplay=15, Front End=0)) {module=CameraOffset001}
   &AddVelocity(
    Module.Use Falloff Distance=false
    Override Seed=false
    Velocity Mode=::NewEnumerator1
    Velocity Speed=&Multiply_Float(
     A=&Multiply_Float(
      A=&MaskFloatBySpawnGroup(
       Masked Float=&RandomRangeFloat(Minimum=150, Maximum=400)
       Passthrough Float=&RandomRangeFloat(Minimum=50, Maximum=75)
      )
      B=&FN_FrontEndFloatModifier(Gameplay=1, Front End=$System.FrontendScale)
     )
     B=&SelectFloatFromArray(
      Direct Array Index=&MakeIntFromFloat(Float=$User.Variant, Conversion Mode=::NewEnumerator1)
      Float Selection Array=DataInterface() {propertyValues {FloatData [1,1,1,1,3,1]}}
     )
    )
   ) {module=AddVelocity001}
  ]
  particleUpdateScript [
   &ParticleState()
   &AccelerationForce(Acceleration=&Multiply_VectorByFloat(
     Vector=&MaskVectorBySpawnGroup(Masked Vector=<0 0 -150>, Passthrough Vector=<0 0 1>)
     Float=&FN_FrontEndFloatModifier(Gameplay=1, Front End=$System.FrontendScale)
    ))
   &Drag(
    Drag=&Multiply_Float(
     A=&MaskFloatBySpawnGroup(Masked Float=6, Passthrough Float=8)
     B=&MergeHelper_FloatSwitch(
      0=1
      1=1
      2=1
      3=1
      4=0.4
      5=1
      Index=&MakeIntFromFloat(Float=$User.Variant, Conversion Mode=::NewEnumerator1)
      Num=::NewEnumerator4
     )
    )
    Use Linear Drag=true
    Use Rotational Drag=true
   )
   &CurlNoiseForce(
    Bias Noise=false
    Noise Strength=&MaskFloatBySpawnGroup(Masked Float=100, Passthrough Float=100)
    Pan Noise=false
   )
   &SolveForcesAndVelocity(Clamp Velocity=false, Limit Acceleration=false)
   &ScaleSpriteSize(
    Uniform Curve Scale=&FN_FrontEndFloatModifier(Gameplay=1, Front End=0.65)
    Uniform Curve Sprite Scale=DataInterface() {propertyValues {Curve {keys [
        {arriveTangent=9.0725, interpMode=RCIM_Cubic, leaveTangent=3.7536, tangentMode=RCTM_Break, time=0, value=0}
        <at 0.5 1 RCIM_Cubic>
        {arriveTangent=-0.0405, interpMode=RCIM_Cubic, leaveTangent=-0.0405, tangentMode=RCTM_User, time=1, value=0}
       ]}}}
   )
   &SubUVAnimation(
    Play Rate:NiagaraFloat=&RandomRangeFloat(Minimum=0.5, Maximum=1)
    Playback Scale Curve=DataInterface() {propertyValues {Curve {keys [
        {interpMode=RCIM_Cubic, tangentMode=RCTM_User, time=0.25, value=1}
        {arriveTangent=-1.2656, interpMode=RCIM_Cubic, leaveTangent=-1.2656, tangentMode=RCTM_User, time=1, value=0}
       ]}}}
    Random Start Frame=true
    "Sprite Renderer "=DataInterface() {propertyValues {SpriteRenderer=&self:NiagaraEmitter_4.NiagaraSpriteRendererProperties_4}}
    SubUV Animation Mode:ENiagara_SubUVLookupModeV2=::NewEnumerator3
    UseEndFrame=false
    UsePlayRateScale=false
    UseStartFrame=false
   )
   &ScaleColor(
    Scale Alpha=&MakeFloatFromLinearColor(LinearColor=<rgba 1 1 1 1>, Channel:ENiagaraLinearColor_Channels=::NewEnumerator3)
    Scale RGB=&VectorFromFloat(Value=&FNTimeOfDayScalar(Night Scale=0.65, Morning Scale=0.85, Evening Scale=0.77))
    ScaleA=true
    ScaleRGB=true
   )
  ]
  rendererClasses [&NiagaraSpriteRendererProperties]
  renderers [
   {rendererClass=&NiagaraSpriteRendererProperties}
   {rendererClass=&NiagaraSpriteRendererProperties, rendererIndex=1}
   {rendererClass=&NiagaraSpriteRendererProperties, rendererIndex=2}
   {rendererClass=&NiagaraSpriteRendererProperties, rendererIndex=3}
  ]
 }
}
summary {
 systemName=NS_Pickaxe_CyberDelivery_Impact
 emitters [
  {
   emitterName=Flash_Elements
   rendererClasses [&NiagaraSpriteRendererProperties, &NiagaraLightRendererProperties]
  }
  {emitterName=Impact_Smoke, rendererClasses [&NiagaraSpriteRendererProperties]}
  {emitterName=DirectionalFlare, rendererClasses [&NiagaraSpriteRendererProperties]}
  {emitterName=DirectionalStreaks, rendererClasses [&NiagaraSpriteRendererProperties]}
  {
   emitterName=DirectionalSparks
   rendererClasses [&NiagaraSpriteRendererProperties, &NiagaraMeshRendererProperties]
  }
  {emitterName=NE_Dust, rendererClasses [&NiagaraSpriteRendererProperties]}
 ]
 userVariables [
  User.bIsFrontEnd {type=NiagaraBool}
  User.bIsFrontEndPreview {type=NiagaraBool}
  User.Direction {type=Vector3f, description="", defaultValue {struct=Vector3f, value=<0 0 0>}}
  User.PhysicalSurface {type=NiagaraInt32}
  User.SwingSectionIndex {type=NiagaraInt32}
  User.Variant {type=NiagaraFloat, defaultValue=0, description=""}
 ]
}
