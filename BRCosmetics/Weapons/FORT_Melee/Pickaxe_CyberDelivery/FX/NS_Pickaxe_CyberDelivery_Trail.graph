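# Niagara swing trail effect for the CyberDelivery pickaxe, with looks selected by User.Variant.
# Trail, TrailTip and TrailGalaxy build anim-trail ribbons between skeletal mesh bones or sockets, fading with SwingGeneralFade.
# Symbols spawns by distance moved, adding curl-noise-driven sprites and a ribbon, with zero size and lifetime for variants 0-1.
# Variant switches drive trail colors and symbol color, size, lifetime and spacing; time-of-day scalars dim brightness.
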NiagaraSystem /BRCosmetics/Weapons/FORT_Melee/Pickaxe_CyberDelivery/FX/NS_Pickaxe_CyberDelivery_Trail
dependencies {
 usedDataInterfaces [
  &NiagaraDataInterfaceColorCurve
  &NiagaraDataInterfaceCurve
  &NiagaraDataInterfaceArrayFloat
  &NiagaraDataInterfaceSkeletalMesh
 ]
 usedDynamicInputs [&SetBoolByIntComparison]
 usedModules [
  &self:SystemSpawnScript.SystemScriptSource.SystemScriptGraph.NiagaraNodeAssignment_1.SetVariables_#1
  &self:SystemSpawnScript.SystemScriptSource.SystemScriptGraph.NiagaraNodeAssignment_2.SetVariables_#2
  &self:SystemSpawnScript.SystemScriptSource.SystemScriptGraph.NiagaraNodeAssignment_3.SetVariables_#3
  &self:SystemSpawnScript.SystemScriptSource.SystemScriptGraph.NiagaraNodeAssignment_4.SetVariables_#4
  &self:SystemSpawnScript.SystemScriptSource.SystemScriptGraph.NiagaraNodeAssignment_5.SetVariables_#5
  &SystemState
  &self:SystemSpawnScript.SystemScriptSource.SystemScriptGraph.NiagaraNodeAssignment_0.SetVariables_#6
 ]
}
emitters {
 Symbols {
  emitterUpdateScript [
   &EmitterState()
   &SpawnRate(
    SpawnRate=&MakeCustomFloatFromBool(Bool=$User.Debug, True Float=10, False Float=0)
    Use Spawn Probability=false
   )
   &SpawnPerUnit(
    Max Movement Threshold=4000
    Module.Use Max Movement Threshold=true
    Module.Use Movement Tolerance=false
    Spawn Probability=1
    Spawn Spacing:NiagaraFloat=&SelectFloatFromArray(
     Direct Array Index=&MakeIntFromFloat(Float=$User.Variant)
     Float Selection Array=DataInterface() {propertyValues {FloatData [0,0,37,37,37,19]}}
    )
    Use Spawn Probability=true
   )
  ]
  particleSpawnScript [
   &InitializeParticle(
    Color=&MergeHelper_ColorSwitch(
     0=<rgba 0 0 0 0>
     1=<rgba 0 0 0 0>
     2=&ColorFromCurve(
      ColorCurve=DataInterface() {propertyValues {
        AlphaCurve {keys [<at 0 1 RCIM_Cubic>]}
        BlueCurve {keys [<at 0.4274 0 RCIM_Cubic>,<at 0.4699 0 RCIM_Cubic>]}
        GreenCurve {keys [<at 0.4274 0.6232 RCIM_Cubic>,<at 0.4699 0.3648 RCIM_Cubic>]}
        RedCurve {keys [<at 0.4274 1 RCIM_Cubic>,<at 0.4699 1 RCIM_Cubic>]}
       }}
      CurveIndex=&RandomRangeFloat(Minimum=0, Maximum=1)
     )
     3=&ColorFromCurve(
      ColorCurve=DataInterface() {propertyValues {
        AlphaCurve {keys [<at 0 1 RCIM_Cubic>]}
        BlueCurve {keys [<at 0.4274 1 RCIM_Cubic>,<at 0.4699 1 RCIM_Cubic>]}
        GreenCurve {keys [<at 0.4274 0.7649 RCIM_Cubic>,<at 0.4699 0.3664 RCIM_Cubic>]}
        RedCurve {keys [<at 0.4274 0.3886 RCIM_Cubic>,<at 0.4699 0.7831 RCIM_Cubic>]}
       }}
      CurveIndex=&RandomRangeFloat(Minimum=0, Maximum=1)
     )
     4=&ColorFromCurve(
      ColorCurve=DataInterface() {propertyValues {
        AlphaCurve {keys [<at 0 0.65 RCIM_Cubic>]}
        BlueCurve {keys [<at 0.4274 1 RCIM_Cubic>,<at 0.4699 1 RCIM_Cubic>]}
        GreenCurve {keys [<at 0.4274 0.9469 RCIM_Cubic>,<at 0.4699 0.6167 RCIM_Cubic>]}
        RedCurve {keys [<at 0.4274 0.5208 RCIM_Cubic>,<at 0.4699 0.4375 RCIM_Cubic>]}
       }}
      CurveIndex=&RandomRangeFloat(Minimum=0, Maximum=1)
     )
     5=&ColorFromCurve(
      ColorCurve=DataInterface() {propertyValues {
        AlphaCurve {keys [<at 0 1 RCIM_Cubic>]}
        BlueCurve {keys [<at 0.4274 1 RCIM_Cubic>,<at 0.4699 1 RCIM_Cubic>]}
        GreenCurve {keys [<at 0.4274 0 RCIM_Cubic>,<at 0.4699 0.282 RCIM_Cubic>]}
        RedCurve {keys [<at 0.4274 0.6698 RCIM_Cubic>,<at 0.4699 0 RCIM_Cubic>]}
       }}
      CurveIndex=&RandomRangeFloat(Minimum=0, Maximum=1)
     )
     Index=&MakeIntFromFloat(Float=$User.Variant, Conversion Mode=::NewEnumerator1)
     Num=::NewEnumerator4
    )
    Lifetime=&MergeHelper_FloatSwitch(
     0=0
     1=0
     2=&RandomRangeFloat(Minimum=0.25, Maximum=0.35)
     3=&RandomRangeFloat(Minimum=0.35, Maximum=0.45)
     4=&RandomRangeFloat(Minimum=0.25, Maximum=0.35)
     5=0.15
     Index=&MakeIntFromFloat(Float=$User.Variant, Conversion Mode=::NewEnumerator1)
     Num=::NewEnumerator4
    )
    Ribbon Width=1
    Ribbon Width Mode=::NewEnumerator1
    Sprite Rotation Angle=&MergeHelper_FloatSwitch(
     0=0
     1=0
     2=&RandomRangeFloat(Minimum=0, Maximum=1)
     3=&RandomRangeFloat(Minimum=0, Maximum=1)
     4=&RandomRangeFloat(Minimum=0, Maximum=1)
     5=0
     Index=&MakeIntFromFloat(Float=$User.Variant, Conversion Mode=::NewEnumerator1)
     Num=::NewEnumerator4
    )
    Sprite Rotation Mode=::NewEnumerator1
    Sprite Size Mode=::NewEnumerator3
    Uniform Sprite Size=&MergeHelper_FloatSwitch(
     0=0
     1=0
     2=&RandomRangeFloat(Minimum=10, Maximum=14)
     3=&RandomRangeFloat(Minimum=3, Maximum=5)
     4=&RandomRangeFloat(Minimum=5, Maximum=8)
     5=&RandomRangeFloat(Minimum=15, Maximum=20)
     Index=&MakeIntFromFloat(Float=$User.Variant, Conversion Mode=::NewEnumerator1)
     Num=::NewEnumerator4
    )
    UsePositionOffset=true
    Write Material Random=true
   )
   &AddVelocity(Velocity=&MakeCustomVectorFromBool(Bool=$User.Debug, True Vector=<350 0 0>, False Vector=<0 0 0>))
   &SkeletalMeshLocation(
    Bone / Socket Index=1
    MaskSpawnGroup=false
    Mesh Sampling Type=::NewEnumerator5
    Sampled Position Offset=&RandomRangeVector(Minimum=<0 -5 0>, Maximum=<0 5 0>)
    Skeletal Mesh=$System.Skeletal Mesh
    Socket Sampling Mode=::NewEnumerator5
   )
   &AddVelocity(
    Rotation Coordinate Space=::Simulation
    Velocity=&RotateVectorByQuaternion(Vector To Rotate=<0 0 1>, Quaternion=$Particles.SkeletalMeshLocation.SampledRotation)
    Velocity Speed Scale=&MakeCustomFloatFromBool(Bool=$User.Debug, True Float=0, False Float=&RandomRangeFloat(Minimum=75, Maximum=150))
   ) {module=AddVelocity001}
   /SetVariables(Particles.SubImageIndexAlt:NiagaraFloat=&RandomRangeFloat(Minimum=0, Maximum=36)) {isSetParameters=true}
  ]
  particleUpdateScript [
   &ParticleState()
   &SkeletalMeshLocation(
    Bone / Socket Index=1
    MaskSpawnGroup=false
    Mesh Sampling Type=::NewEnumerator5
    Position Sampling=::NewEnumerator2
    Skeletal Mesh=$System.Skeletal Mesh
    Socket Sampling Mode=::NewEnumerator5
   ) {module=SkeletalMeshLocation001}
   &ScaleSpriteSize(Uniform Curve Sprite Scale=DataInterface() {propertyValues {Curve {keys [
        {arriveTangent=8.2127, interpMode=RCIM_Cubic, leaveTangent=8.2127, tangentMode=RCTM_User, time=0, value=0}
        {
         arriveTangent=0.0908
         arriveTangentWeight=0.232
         interpMode=RCIM_Cubic
         leaveTangent=0.0908
         leaveTangentWeight=2.2538
         tangentMode=RCTM_User
         time=0.1
         value=1
        }
        {arriveTangent=-0.042, interpMode=RCIM_Cubic, leaveTangent=-0.042, tangentMode=RCTM_User, time=1, value=0}
       ]}}}) {module=ScaleSpriteSize001}
   &Drag(Drag=&RandomRangeFloat(Minimum=6, Maximum=8), Use Linear Drag=true, Use Rotational Drag=true)
   &CurlNoiseForce(
    Bias Noise=false
    Noise Strength=&FloatFromCurve(
     FloatCurve=DataInterface() {propertyValues {Curve {keys [
         {arriveTangent=2.7279, interpMode=RCIM_Cubic, leaveTangent=2.7279, tangentMode=RCTM_User, time=0, value=0}
         {interpMode=RCIM_Cubic, tangentMode=RCTM_User, time=1, value=1}
        ]}}}
     Scale Curve=550
    )
    Pan Noise=false
   )
   &ScaleRibbonWidth(Ribbon Width Scale=&FloatFromCurve(FloatCurve=DataInterface() {propertyValues {Curve {keys [{interpMode=RCIM_Cubic, tangentMode=RCTM_User, time=0, value=1}, {interpMode=RCIM_Cubic, tangentMode=RCTM_User, time=1, value=0}]}}}))
   &SolveForcesAndVelocity(Clamp Velocity=false, Limit Acceleration=false)
   &ScaleColor(
    Scale RGB=&VectorFromFloat(Value=&FNTimeOfDayScalar(Night Scale=0.7, Morning Scale=0.8, Evening Scale=0.9))
    ScaleA=true
    ScaleRGB=true
   )
  ]
  rendererClasses [&NiagaraSpriteRendererProperties, &NiagaraRibbonRendererProperties]
  renderers [
   {rendererClass=&NiagaraSpriteRendererProperties}
   {rendererClass=&NiagaraSpriteRendererProperties, rendererIndex=1}
   {rendererClass=&NiagaraSpriteRendererProperties, rendererIndex=2}
   {rendererClass=&NiagaraSpriteRendererProperties, rendererIndex=3}
   {rendererClass=&NiagaraRibbonRendererProperties, rendererIndex=4}
  ]
 }
 Trail {
  emitterUpdateScript [
   &EmitterState()
   &FN_SpawnAnimTrail_v2()
  ]
  particleSpawnScript [
   /Game/Effects/Niagara/Modules/AnimTrail/FN_InitializeAnimTrail/FN_InitializeAnimTrailBoneIndecies(Second Bone or Socket Index=1)
   &FN_InitializeAnimTrail(Color=<rgba 2 2 2 1>, Lifetime=0.1, SkeletalMesh=$System.Skeletal Mesh) {module=FN_InitializeAnimTrail001}
  ]
  particleUpdateScript [
   &FN_UpdateAnimTrail(SkeletalMesh=$System.Skeletal Mesh)
   /SetVariables(
    Particles.HexTrail:LinearColor=&MultiplyLinearColor(
     Linear Color=&MergeHelper_ColorSwitch(
      0=<rgba 1 0.2928 0.2812 1>
      1=<rgba 1 0.2928 0.2812 1>
      2=&ColorFromCurve(ColorCurve=DataInterface() {propertyValues {
         AlphaCurve {keys [<at 0 1 RCIM_Cubic>,<at 1 1 RCIM_Cubic>]}
         BlueCurve {keys [<at 0 0.2993 RCIM_Cubic>,<at 1 0.3074 RCIM_Cubic>]}
         GreenCurve {keys [<at 0 0.4865 RCIM_Cubic>,<at 1 0.3648 RCIM_Cubic>]}
         RedCurve {keys [<at 0 1 RCIM_Cubic>,<at 1 1 RCIM_Cubic>]}
        }})
      3=&ColorFromCurve(
       ColorCurve=DataInterface() {propertyValues {
         AlphaCurve {keys [<at 0 1 RCIM_Cubic>,<at 1 1 RCIM_Cubic>]}
         BlueCurve {keys [<at 0 1 RCIM_Cubic>,<at 0.149 1>,<at 0.4187 1>,<at 0.7148 1>,<at 1 0.8261 RCIM_Cubic>]}
         GreenCurve {keys [<at 0 0.8818 RCIM_Cubic>,<at 0.149 0.3306>,<at 0.4187 0.3438>,<at 0.7148 0.0677>,<at 1 1 RCIM_Cubic>]}
         RedCurve {keys [<at 0 0.3673 RCIM_Cubic>,<at 0.149 0.243>,<at 0.4187 0.509>,<at 0.7148 0.7464>,<at 1 0.3062 RCIM_Cubic>]}
        }}
       CurveIndex=&Divide_Float(A=$Particles.NormalizedAge, B=1)
      )
      4=<rgba 0.0134 1 0.8271 1>
      5=<rgba 0.233 0.0157 0.8385 1>
      Index=&MakeIntFromFloat(Float=$User.Variant, Conversion Mode=::NewEnumerator1)
      Num=::NewEnumerator4
     )
     Scale Factor=&MakeFloat4FromVectorAndFloat(
      "Vector (RGB)"=&VectorFromFloat(Value=&FNTimeOfDayScalar(Night Scale=0.63, Morning Scale=0.8, Evening Scale=0.7))
      "Float (Alpha)"=&FloatFromCurve(
       FloatCurve=DataInterface() {propertyValues {Curve {keys [
           {arriveTangent=-3.1634, interpMode=RCIM_Cubic, leaveTangent=-3.1634, tangentMode=RCTM_User, time=0, value=1}
           {arriveTangent=-0.0789, tangentMode=RCTM_User, time=1, value=0}
          ]}}}
       Scale Curve=&Multiply_Float(A=0.6, B=$System.SwingGeneralFade)
      )
     )
    )
    Particles.HexTrailWidth:NiagaraFloat=&Multiply_Float(A=$Particles.RibbonWidth, B=1.1)
   ) {isSetParameters=true}
   /SetVariables(
    Particles.SoftTrailColor:LinearColor=&MultiplyLinearColor(
     Linear Color=&MergeHelper_ColorSwitch(
      0=<rgba 1.3 0.0676 0.0474 1>
      1=<rgba 1.3 0.0676 0.0474 1>
      2=&ColorFromCurve(ColorCurve=DataInterface() {propertyValues {
         AlphaCurve {keys [<at 0 1 RCIM_Cubic>,<at 1 1 RCIM_Cubic>]}
         BlueCurve {keys [<at 0 0.0701 RCIM_Cubic>,<at 1 0.101 RCIM_Cubic>]}
         GreenCurve {keys [<at 0 0.3464 RCIM_Cubic>,<at 1 0.2297 RCIM_Cubic>]}
         RedCurve {keys [<at 0 1.3 RCIM_Cubic>,<at 1 1.3 RCIM_Cubic>]}
        }})
      3=&ColorFromCurve(
       ColorCurve=DataInterface() {propertyValues {
         AlphaCurve {keys [<at 0 1 RCIM_Cubic>,<at 1 1 RCIM_Cubic>]}
         BlueCurve {keys [<at 0 1.3 RCIM_Cubic>,<at 0.149 1.3>,<at 0.4187 1.3>,<at 0.7148 1.3>,<at 1 1.074 RCIM_Cubic>]}
         GreenCurve {keys [<at 0 1.1463 RCIM_Cubic>,<at 0.149 0.4298>,<at 0.4187 0.4469>,<at 0.7148 0.088>,<at 1 1.3 RCIM_Cubic>]}
         RedCurve {keys [<at 0 0.4775 RCIM_Cubic>,<at 0.149 0.3159>,<at 0.4187 0.6617>,<at 0.7148 0.9704>,<at 1 0.3981 RCIM_Cubic>]}
        }}
       CurveIndex=&Divide_Float(A=$Particles.NormalizedAge, B=1)
      )
      4=&ColorFromCurve(ColorCurve=DataInterface() {propertyValues {
         AlphaCurve {keys [<at 0 1 RCIM_Cubic>,<at 1 1 RCIM_Cubic>]}
         BlueCurve {keys [<at 0 1.3 RCIM_Cubic>,<at 1 0.8412 RCIM_Cubic>]}
         GreenCurve {keys [<at 0 0.9415 RCIM_Cubic>,<at 1 1.3 RCIM_Cubic>]}
         RedCurve {keys [<at 0 0.0432 RCIM_Cubic>,<at 1 0.0017 RCIM_Cubic>]}
        }})
      5=&ColorFromCurve(ColorCurve=DataInterface() {propertyValues {
         AlphaCurve {keys [<at 0 1 RCIM_Cubic>,<at 1 1 RCIM_Cubic>]}
         BlueCurve {keys [<at 0 0.5252 RCIM_Cubic>,<at 0.514 0.7031 RCIM_Cubic>]}
         GreenCurve {keys [<at 0 0.0477 RCIM_Cubic>,<at 0.514 0.5544 RCIM_Cubic>]}
         RedCurve {keys [<at 0 0.5729 RCIM_Cubic>,<at 0.514 0.1023 RCIM_Cubic>]}
        }})
      Index=&MakeIntFromFloat(Float=$User.Variant, Conversion Mode=::NewEnumerator1)
      Num=::NewEnumerator4
     )
     Scale Factor=&MakeFloat4FromVectorAndFloat(
      "Vector (RGB)"=&VectorFromFloat(Value=&FNTimeOfDayScalar(Night Scale=0.63, Morning Scale=0.8, Evening Scale=0.7))
      "Float (Alpha)"=&FloatFromCurve(
       FloatCurve=DataInterface() {propertyValues {Curve {keys [<at 0 1 RCIM_Cubic>,<at 1 0>]}}}
       Scale Curve=&Multiply_Float(A=0.5, B=$System.SwingGeneralFade)
      )
     )
    )
    Particles.SoftTrailWidth:NiagaraFloat=&Multiply_Float(A=$Particles.RibbonWidth, B=1)
   ) {isSetParameters=true, module=SetVariables_#7}
  ]
  rendererClasses [&NiagaraRibbonRendererProperties]
  renderers [
   {rendererClass=&NiagaraRibbonRendererProperties}
   {rendererClass=&NiagaraRibbonRendererProperties, rendererIndex=1}
  ]
 }
 TrailGalaxy {
  emitterUpdateScript [
   &EmitterState()
   &FN_SpawnAnimTrail()
  ]
  particleSpawnScript [
   /Game/Effects/Niagara/Modules/AnimTrail/FN_InitializeAnimTrailBoneIndecies_v2/FN_InitializeAnimTrailBoneIndecies(First Bone or Socket Index=1, Second Bone or Socket Index=0)
   &FN_InitializeAnimTrail(Lifetime=0.2, SkeletalMesh=$System.Skeletal Mesh)
   &DynamicMaterialParameters(
    Index 0 Param 2=1
    Index 0 Param 3=5
    Index 0 Param 4=0.35
    Index 1 Param 1=0.25
    Index 1 Param 2=0.464
    Index 1 Param 3=0.088
    Module.Index0Float0Write=true
    Module.Index0Float1Write=true
    Module.Index0Float2Write=true
    Module.Index0Float3Write=true
    Module.Index1Float0Write=true
    Module.Index1Float1Write=true
    Module.Index1Float2Write=true
    Module.Index1Float3Write=true
    Write Parameter Index 1=::NewEnumerator1
   )
  ]
  particleUpdateScript [
   &ParticleState()
   &FN_UpdateAnimTrail(SkeletalMesh=$System.Skeletal Mesh)
   &ScaleColor(
    Scale Alpha=&FloatFromCurve(
     FloatCurve=DataInterface() {propertyValues {Curve {keys [
         {arriveTangent=-0.087, interpMode=RCIM_Cubic, leaveTangent=-0.087, tangentMode=RCTM_User, time=0.3538, value=2}
         {interpMode=RCIM_Cubic, tangentMode=RCTM_User, time=1, value=0}
        ]}}}
     Scale Curve=&Multiply_Float(
      A=&FNTimeOfDayScalar(Night Scale=0.7, Morning Scale=0.9, Evening Scale=0.8)
      B=$System.SwingGeneralFade
     )
    )
    Scale RGB=&MakeVectorFromLinearColorRGB(Color=<rgba 1 1 1 1>)
    ScaleA=true
    ScaleRGB=true
   ) {module=ScaleColor001}
  ]
  rendererClasses [&NiagaraRibbonRendererProperties]
  renderers [{rendererClass=&NiagaraRibbonRendererProperties}]
 }
 TrailTip {
  emitterUpdateScript [
   &EmitterState()
   &FN_SpawnAnimTrail_v2()
  ]
  particleSpawnScript [
   /Game/Effects/Niagara/Modules/AnimTrail/FN_InitializeAnimTrail/FN_InitializeAnimTrailBoneIndecies(First Bone or Socket Index=2, Second Bone or Socket Index=1)
   &FN_InitializeAnimTrail(AnimTrailWidth=1.02, Color=<rgba 2 2 2 1>, Lifetime=0.1, SkeletalMesh=$System.Skeletal Mesh) {module=FN_InitializeAnimTrail001}
  ]
  particleUpdateScript [
   &FN_UpdateAnimTrail(SkeletalMesh=$System.Skeletal Mesh)
   /SetVariables(
    Particles.SoftTrailColor:LinearColor=&MultiplyLinearColor(
     Linear Color=&MergeHelper_ColorSwitch(
      0=<rgba 7 0.3639 0.2552 1>
      1=<rgba 7 0.3639 0.2552 1>
      2=&ColorFromCurve(ColorCurve=DataInterface() {propertyValues {
         AlphaCurve {keys [<at 0 1 RCIM_Cubic>,<at 1 1 RCIM_Cubic>]}
         BlueCurve {keys [<at 0 0.1531 RCIM_Cubic>,<at 1 0.1526 RCIM_Cubic>]}
         GreenCurve {keys [<at 0 1.4769 RCIM_Cubic>,<at 1 0.6453 RCIM_Cubic>]}
         RedCurve {keys [<at 0 5 RCIM_Cubic>,<at 1 5 RCIM_Cubic>]}
        }})
      3=&ColorFromCurve(ColorCurve=DataInterface() {propertyValues {
         AlphaCurve {keys [<at 0 1 RCIM_Cubic>,<at 1 1 RCIM_Cubic>]}
         BlueCurve {keys [<at 0 5 RCIM_Cubic>,<at 0.149 5>,<at 0.4187 5>,<at 0.7148 5>,<at 1 4.1306 RCIM_Cubic>]}
         GreenCurve {keys [<at 0 4.2311 RCIM_Cubic>,<at 0.149 1.0163>,<at 0.4187 0.625>,<at 0.7148 0.3385>,<at 1 5 RCIM_Cubic>]}
         RedCurve {keys [<at 0 0.8854 RCIM_Cubic>,<at 0.149 0.4948>,<at 0.4187 1.7268>,<at 0.7148 3.7322>,<at 1 1.5312 RCIM_Cubic>]}
        }})
      4=&ColorFromCurve(ColorCurve=DataInterface() {propertyValues {
         AlphaCurve {keys [<at 0 1 RCIM_Cubic>,<at 1 1 RCIM_Cubic>]}
         BlueCurve {keys [<at 0 3.8 RCIM_Cubic>,<at 1 2.1243 RCIM_Cubic>]}
         GreenCurve {keys [<at 0 2.7521 RCIM_Cubic>,<at 1 3.8 RCIM_Cubic>]}
         RedCurve {keys [<at 0 0.1261 RCIM_Cubic>,<at 1 0 RCIM_Cubic>]}
        }})
      5=&ColorFromCurve(ColorCurve=DataInterface() {propertyValues {
         AlphaCurve {keys [<at 0 1 RCIM_Cubic>,<at 1 1 RCIM_Cubic>]}
         BlueCurve {keys [<at 0.2403 6 RCIM_Cubic>,<at 0.7695 2.8751 RCIM_Cubic>]}
         GreenCurve {keys [<at 0.2403 0 RCIM_Cubic>,<at 0.7695 3 RCIM_Cubic>]}
         RedCurve {keys [<at 0.2403 0.9799 RCIM_Cubic>,<at 0.7695 0 RCIM_Cubic>]}
        }})
      Index=&MakeIntFromFloat(Float=$User.Variant, Conversion Mode=::NewEnumerator1)
      Num=::NewEnumerator4
     )
     Scale Factor=&MakeFloat4FromVectorAndFloat(
      "Vector (RGB)"=&VectorFromFloat(Value=&MergeHelper_FloatSwitch(
        0=&FNTimeOfDayScalar(Night Scale=0.63, Morning Scale=0.8, Evening Scale=0.7)
        1=&FNTimeOfDayScalar(Night Scale=0.63, Morning Scale=0.8, Evening Scale=0.7)
        2=&FNTimeOfDayScalar(Night Scale=0.63, Morning Scale=0.8, Evening Scale=0.7)
        3=&FNTimeOfDayScalar(Night Scale=0.53, Morning Scale=0.8, Evening Scale=0.7)
        4=&FNTimeOfDayScalar(Night Scale=0.63, Morning Scale=0.8, Evening Scale=0.7)
        5=&FNTimeOfDayScalar(Night Scale=0.63, Morning Scale=0.8, Evening Scale=0.7)
        Index=&MakeIntFromFloat(Float=$User.Variant, Conversion Mode=::NewEnumerator1)
        Num=::NewEnumerator4
       ))
      "Float (Alpha)"=&FloatFromCurve(
       FloatCurve=DataInterface() {propertyValues {Curve {keys [<at 0 1 RCIM_Cubic>,<at 1 0>]}}}
       Scale Curve=&Multiply_Float(A=0.5, B=$System.SwingGeneralFade)
      )
     )
    )
    Particles.SoftTrailWidth:NiagaraFloat=&Multiply_Float(A=$Particles.RibbonWidth, B=1)
   ) {isSetParameters=true, module=SetVariables_#7}
  ]
  rendererClasses [&NiagaraRibbonRendererProperties]
  renderers [{rendererClass=&NiagaraRibbonRendererProperties}]
 }
}
summary {
 systemName=NS_Pickaxe_CyberDelivery_Trail
 emitters [
  {emitterName=Trail, rendererClasses [&NiagaraRibbonRendererProperties]}
  {emitterName=TrailTip, rendererClasses [&NiagaraRibbonRendererProperties]}
  {emitterName=TrailGalaxy, rendererClasses [&NiagaraRibbonRendererProperties]}
  {emitterName=Symbols, rendererClasses [&NiagaraSpriteRendererProperties, &NiagaraRibbonRendererProperties]}
 ]
 userVariables [
  User.Debug {type=NiagaraBool}
  User.RibbonIDOffset {type=NiagaraInt32}
  User.SwingSectionIndex {type=NiagaraInt32, defaultValue=1}
  User.Variant {type=NiagaraFloat, defaultValue=0, description=""}
 ]
}
